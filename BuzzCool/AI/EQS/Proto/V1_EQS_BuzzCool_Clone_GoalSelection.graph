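# EQS goal-selection query (prototype) for the BuzzCool clone AI, trying six options in priority order.
# Recall-tagged players come first, then perceived enemies favoring marked targets, marked players, marked composite-generated actors, deployed BuzzCool items, and finally the item owner.
# Enemy and marked-target options score perception, facing, proximity and pathfinding, discourage frustrated goals, and filter by 2D distance from the clone's leash center.
# Recall and owner options are distance-limited.
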EnvQuery /BuzzCool/AI/EQS/Proto/V1_EQS_BuzzCool_Clone_GoalSelection
options [
 {
  generator {
   name=FortQueryGenerator_GoalPlayerPawns_2
   class=FortQueryGenerator_GoalPlayerPawns
   properties {bAutoSortTests=true, bCanRunAsync=false, bOnlyAthenaGameParticipants=true, envQueryResultNormalizationOption=Default, optionName=Item Recall}
  }
  tests [
   FortQueryTest_GoalGameplayTags_3 {
    class=FortQueryTest_GoalGameplayTags
    properties {
     bDefineReferenceValue=false
     bRequireAllProvidedTagQueriesPass=false
     bScoreEncounterGoals=true
     bScoreEnemies=true
     bScoreSpecificAssignments=false
     bScoreWorldGoals=true
     bShouldLookupQueryByTag=false
     bShouldPassWhenQueryNotFound=true
     boolValue=true
     clampMaxType=None
     clampMinType=None
     filterType=Match
     floatValueMax=0
     floatValueMin=0
     multipleContextFilterOp=AllPass
     multipleContextScoreOp=AverageScore
     normalizationType=Absolute
     referenceValue=0
     scoreClampMax=0
     scoreClampMin=0
     scoringEquation=Constant
     scoringFactor=1
     testComment=""
     testPurpose=Filter
     assignmentIds []
     goalActorTagQuery {autoDescription="", tokenStreamVersion=0, userDescription="", queryTokenStream [], tagDictionary []}
     pawnOwnerTagQuery {autoDescription="", tokenStreamVersion=0, userDescription="", queryTokenStream [], tagDictionary []}
     prohibitedAssignmentIds []
     queryLookupTag {tagName=None}
     tagQueryToMatch {
      autoDescription=" ANY_EXACT( Item.BuzzCool.Recall )"
      tokenStreamVersion=0
      userDescription=""
      queryTokenStream [0,1,7,1,0]
      tagDictionary [{tagName=Item.BuzzCool.Recall}]
     }
    }
   }
   FortQueryTest_GoalDistance_8 {
    class=FortQueryTest_GoalDistance
    properties {
     bScoreEncounterGoals=true
     bScoreEnemies=true
     bScoreSpecificAssignments=false
     bScoreWorldGoals=true
     clampMaxType=None
     clampMinType=None
     filterType=Maximum
     scoreClampMax=0
     scoreClampMin=0
     scoringEquation=Linear
     scoringFactor=1
     testComment=""
     testPurpose=Filter
     assignmentIds []
     floatValueMax {
      dataBinding=&self:FortQueryTest_GoalDistance_8.FortAIDataProvider_ScalableFloat_0
      dataField=FloatValue
      defaultValue=0
     }
    }
   }
  ]
 }
 {
  generator {
   name=FortQueryGenerator_Enemies_0
   class=FortQueryGenerator_Enemies
   properties {
    bAddEnemiesFromAbilityRange=false
    bAutoSortTests=true
    bCanRunAsync=false
    bIgnoreDBNOPawns=true
    bIgnoreSleepingAIs=false
    bIncludeMarkedSleepingAIs=false
    bIncludePropagatedEnemies=false
    bPerceivedEnemiesOnly=true
    bSleepCapableAIUsePerceivedEnemiesOnly=true
    envQueryResultNormalizationOption=Default
    optionName=""
    abilityTags {gameplayTags []}
    distanceTags {gameplayTags []}
    maxTimeSincePerceived {
     dataBinding=&self:FortQueryGenerator_Enemies_0.FortAIDataProvider_ScalableFloat_0
     dataField=FloatValue
     defaultValue=15
    }
   }
  }
  tests [
   EnvQueryTest_GameplayTags_0 {
    class=EnvQueryTest_GameplayTags
    properties {
     bDefineReferenceValue=false
     bRejectIncompatibleItems=false
     bSearchComponentsForTags=false
     boolValue=true
     clampMaxType=None
     clampMinType=None
     filterType=Match
     floatValueMax=0
     floatValueMin=0
     multipleContextFilterOp=AllPass
     multipleContextScoreOp=AverageScore
     normalizationType=Absolute
     referenceValue=0
     scoreClampMax=0
     scoreClampMin=0
     scoringEquation=Constant
     scoringFactor=5
     testComment=Heavily prefer Marked Targets
     testPurpose=Score
     tagQueryToMatch {
      autoDescription=" ANY( Ability.BuzzCool.Item.MarkedTarget )"
      tokenStreamVersion=0
      userDescription=""
      queryTokenStream [0,1,1,1,0]
      tagDictionary [{tagName=Ability.BuzzCool.Item.MarkedTarget}]
     }
    }
   }
   FortQueryTest_PerceptionAll_0 {
    class=FortQueryTest_PerceptionAll
    properties {
     clampMaxType=None
     clampMinType=None
     scoreClampMax=0
     scoringFactor=2
     additionalSenseScores {&Class'&AthenaAISense_Hearing'=1}
     maxSenseAge {
      dataBinding=&self:FortQueryTest_PerceptionAll_0.FortAIDataProvider_ScalableFloat_1
      dataField=FloatValue
      defaultValue=0
     }
     minSenseAge {
      dataBinding=&self:FortQueryTest_PerceptionAll_0.FortAIDataProvider_ScalableFloat_0
      dataField=FloatValue
      defaultValue=0
     }
     senseScores [0,0,0,0,0,0]
    }
   }
   EnvQueryTest_PathfindingBatch_1 {class=EnvQueryTest_PathfindingBatch, properties {filterClass=/Game/Athena/AI/NavFilters/NavFilter_NPC_AllNavmesh.NavFilter_NPC_AllNavmesh_C}}
   FortQueryTest_GoalFrustrationDiscouragement_0 {
    class=FortQueryTest_GoalFrustrationDiscouragement
    properties {
     bDefineReferenceValue=false
     bScoreEncounterGoals=true
     bScoreEnemies=true
     bScoreSpecificAssignments=false
     bScoreWorldGoals=true
     boolValue=true
     clampMaxType=SpecifiedValue
     clampMinType=SpecifiedValue
     filterType=Maximum
     floatValueMax=1
     floatValueMin=0
     multipleContextFilterOp=AllPass
     multipleContextScoreOp=AverageScore
     normalizationType=Absolute
     referenceValue=0
     scoreClampMax=1
     scoreClampMin=0
     scoringEquation=Constant
     scoringFactor=-1
     testComment="When a Goal can't be reached, we can choose to Frustrate on (IGNORE) that Goal for a period of time."
     testPurpose=FilterAndScore
     assignmentIds []
     goalActorTagQuery {autoDescription="", tokenStreamVersion=0, userDescription="", queryTokenStream [], tagDictionary []}
     prohibitedAssignmentIds []
    }
   }
   EnvQueryTest_Dot_5 {
    class=EnvQueryTest_Dot
    properties {
     clampMaxType=SpecifiedValue
     clampMinType=SpecifiedValue
     scoreClampMax=0
     scoreClampMin=0
     testComment=Prefer Goal in front
     lineA {dirMode=TwoPoints, lineFrom=&EnvQueryContext_Querier, lineTo=&EnvQueryContext_Item}
     lineB {dirMode=Rotation, lineTo=&FortQueryContext_Goal, rotation=&EnvQueryContext_Querier}
     scoringFactor {dataBinding=null, dataField=FloatValue, defaultValue=1}
    }
   }
   FortQueryTest_GoalDistance_1 {
    class=FortQueryTest_GoalDistance
    properties {
     bScoreEncounterGoals=true
     bScoreEnemies=true
     bScoreSpecificAssignments=false
     bScoreWorldGoals=true
     clampMaxType=None
     clampMinType=None
     filterType=Range
     floatValueMax=0
     scoreClampMax=0
     scoreClampMin=0
     scoringEquation=InverseLinear
     scoringFactor=1
     testComment=Prefer closer targets.
     testPurpose=Score
     assignmentIds []
    }
   }
   FortQueryTest_GoalDistance_3 {
    class=FortQueryTest_GoalDistance
    properties {
     bScoreEncounterGoals=true
     bScoreEnemies=true
     bScoreSpecificAssignments=false
     bScoreWorldGoals=true
     clampMaxType=None
     clampMinType=None
     distanceTo=/BuzzCool/AI/EQS/EnvQueryContext_BuzzCool_Clone_LeashCenter.EnvQueryContext_BuzzCool_Clone_LeashCenter_C
     filterType=Maximum
     scoreClampMax=0
     scoreClampMin=0
     scoringEquation=Linear
     scoringFactor=1
     testComment=Leash: ignore outside the leash.
     testMode=Distance2D
     testPurpose=Filter
     assignmentIds []
     floatValueMax {
      dataBinding=&self:FortQueryTest_GoalDistance_3.FortAIDataProvider_ScalableFloat_0
      dataField=FloatValue
      defaultValue=0
     }
    }
   }
  ]
 }
 {
  generator {
   name=FortQueryGenerator_GoalPlayerPawns_1
   class=FortQueryGenerator_GoalPlayerPawns
   properties {bAutoSortTests=true, bCanRunAsync=false, bOnlyAthenaGameParticipants=true, envQueryResultNormalizationOption=Default, optionName=""}
  }
  tests [
   EnvQueryTest_GameplayTags_1 {
    class=EnvQueryTest_GameplayTags
    properties {
     bDefineReferenceValue=false
     bRejectIncompatibleItems=false
     bSearchComponentsForTags=false
     boolValue=true
     clampMaxType=None
     clampMinType=None
     filterType=Match
     floatValueMax=0
     floatValueMin=0
     multipleContextFilterOp=AllPass
     multipleContextScoreOp=AverageScore
     normalizationType=Absolute
     referenceValue=0
     scoreClampMax=0
     scoreClampMin=0
     scoringEquation=Constant
     scoringFactor=5
     testComment=Require Marked Targets
     testPurpose=Filter
     tagQueryToMatch {
      autoDescription=" ANY( Ability.BuzzCool.Item.MarkedTarget )"
      tokenStreamVersion=0
      userDescription=""
      queryTokenStream [0,1,1,1,0]
      tagDictionary [{tagName=Ability.BuzzCool.Item.MarkedTarget}]
     }
    }
   }
   FortQueryTest_PerceptionAll_1 {
    class=FortQueryTest_PerceptionAll
    properties {
     clampMaxType=None
     clampMinType=None
     scoreClampMax=0
     scoringFactor=2
     additionalSenseScores {&Class'&AthenaAISense_Hearing'=1}
     maxSenseAge {
      dataBinding=&self:FortQueryTest_PerceptionAll_1.FortAIDataProvider_ScalableFloat_1
      dataField=FloatValue
      defaultValue=0
     }
     minSenseAge {
      dataBinding=&self:FortQueryTest_PerceptionAll_1.FortAIDataProvider_ScalableFloat_0
      dataField=FloatValue
      defaultValue=0
     }
     senseScores [0,0,0,0,0,0]
    }
   }
   EnvQueryTest_PathfindingBatch_2 {class=EnvQueryTest_PathfindingBatch, properties {filterClass=&NavFilter_NPC_AllNavmesh_C}}
   FortQueryTest_GoalFrustrationDiscouragement_2 {
    class=FortQueryTest_GoalFrustrationDiscouragement
    properties {
     bDefineReferenceValue=false
     bScoreEncounterGoals=true
     bScoreEnemies=true
     bScoreSpecificAssignments=false
     bScoreWorldGoals=true
     boolValue=true
     clampMaxType=SpecifiedValue
     clampMinType=SpecifiedValue
     filterType=Maximum
     floatValueMax=1
     floatValueMin=0
     multipleContextFilterOp=AllPass
     multipleContextScoreOp=AverageScore
     normalizationType=Absolute
     referenceValue=0
     scoreClampMax=1
     scoreClampMin=0
     scoringEquation=Constant
     scoringFactor=-1
     testComment="When a Goal can't be reached, we can choose to Frustrate on (IGNORE) that Goal for a period of time."
     testPurpose=FilterAndScore
     assignmentIds []
     goalActorTagQuery {autoDescription="", tokenStreamVersion=0, userDescription="", queryTokenStream [], tagDictionary []}
     prohibitedAssignmentIds []
    }
   }
   EnvQueryTest_Dot_0 {
    class=EnvQueryTest_Dot
    properties {
     clampMaxType=SpecifiedValue
     clampMinType=SpecifiedValue
     scoreClampMax=0
     scoreClampMin=0
     testComment=Prefer Goal in front
     lineA {dirMode=TwoPoints, lineFrom=&EnvQueryContext_Querier, lineTo=&EnvQueryContext_Item}
     lineB {dirMode=Rotation, lineTo=&FortQueryContext_Goal, rotation=&EnvQueryContext_Querier}
     scoringFactor {dataBinding=null, dataField=FloatValue, defaultValue=1}
    }
   }
   FortQueryTest_GoalDistance_2 {
    class=FortQueryTest_GoalDistance
    properties {
     bScoreEncounterGoals=true
     bScoreEnemies=true
     bScoreSpecificAssignments=false
     bScoreWorldGoals=true
     clampMaxType=None
     clampMinType=None
     filterType=Range
     floatValueMax=0
     scoreClampMax=0
     scoreClampMin=0
     scoringEquation=InverseLinear
     scoringFactor=1
     testComment=Prefer closer targets.
     testPurpose=Score
     assignmentIds []
    }
   }
   FortQueryTest_GoalDistance_4 {
    class=FortQueryTest_GoalDistance
    properties {
     bScoreEncounterGoals=true
     bScoreEnemies=true
     bScoreSpecificAssignments=false
     bScoreWorldGoals=true
     clampMaxType=None
     clampMinType=None
     distanceTo=&EnvQueryContext_BuzzCool_Clone_LeashCenter_C
     filterType=Maximum
     scoreClampMax=0
     scoreClampMin=0
     scoringEquation=Linear
     scoringFactor=1
     testComment=Leash: ignore outside the leash.
     testMode=Distance2D
     testPurpose=Filter
     assignmentIds []
     floatValueMax {
      dataBinding=&self:FortQueryTest_GoalDistance_4.FortAIDataProvider_ScalableFloat_0
      dataField=FloatValue
      defaultValue=0
     }
    }
   }
  ]
 }
 {
  generator {
   name=EnvQueryGenerator_Composite_0
   class=EnvQueryGenerator_Composite
   properties {
    optionName=""
    generators [
     &self:EnvQueryGenerator_Composite_0.FortQueryGenerator_GoalActorsOfClass_0
     &self:EnvQueryGenerator_Composite_0.FortQueryGenerator_GoalActorsOfClass_1
     &self:EnvQueryGenerator_Composite_0.FortQueryGenerator_GoalActorsOfClass_3
     &self:EnvQueryGenerator_Composite_0.FortQueryGenerator_GoalActorsOfClass_4
     &self:EnvQueryGenerator_Composite_0.FortQueryGenerator_GoalActorsOfClass_5
     &self:EnvQueryGenerator_Composite_0.FortQueryGenerator_GoalActorsOfClass_2
    ]
   }
  }
  tests [
   FortQueryTest_GoalGameplayTags_1 {
    class=FortQueryTest_GoalGameplayTags
    properties {
     bDefineReferenceValue=false
     bRequireAllProvidedTagQueriesPass=false
     bScoreEncounterGoals=true
     bScoreEnemies=true
     bScoreSpecificAssignments=false
     bScoreWorldGoals=true
     bShouldLookupQueryByTag=false
     bShouldPassWhenQueryNotFound=true
     boolValue=true
     clampMaxType=None
     clampMinType=None
     filterType=Match
     floatValueMax=0
     floatValueMin=0
     multipleContextFilterOp=AllPass
     multipleContextScoreOp=AverageScore
     normalizationType=Absolute
     referenceValue=0
     scoreClampMax=0
     scoreClampMin=0
     scoringEquation=Constant
     scoringFactor=1
     testComment=""
     testPurpose=Filter
     assignmentIds []
     goalActorTagQuery {autoDescription="", tokenStreamVersion=0, userDescription="", queryTokenStream [], tagDictionary []}
     pawnOwnerTagQuery {autoDescription="", tokenStreamVersion=0, userDescription="", queryTokenStream [], tagDictionary []}
     prohibitedAssignmentIds []
     queryLookupTag {tagName=None}
     tagQueryToMatch {
      autoDescription=" ANY( Ability.BuzzCool.Item.MarkedTarget )"
      tokenStreamVersion=0
      userDescription=""
      queryTokenStream [0,1,1,1,0]
      tagDictionary [{tagName=Ability.BuzzCool.Item.MarkedTarget}]
     }
    }
   }
   FortQueryTest_PerceptionAll_2 {
    class=FortQueryTest_PerceptionAll
    properties {
     clampMaxType=None
     clampMinType=None
     scoreClampMax=0
     scoringFactor=2
     additionalSenseScores {&Class'&AthenaAISense_Hearing'=1}
     maxSenseAge {
      dataBinding=&self:FortQueryTest_PerceptionAll_2.FortAIDataProvider_ScalableFloat_1
      dataField=FloatValue
      defaultValue=0
     }
     minSenseAge {
      dataBinding=&self:FortQueryTest_PerceptionAll_2.FortAIDataProvider_ScalableFloat_0
      dataField=FloatValue
      defaultValue=0
     }
     senseScores [1,1,1,0,0,0]
    }
   }
   EnvQueryTest_PathfindingBatch_0 {class=EnvQueryTest_PathfindingBatch, properties {filterClass=&NavFilter_NPC_AllNavmesh_C}}
   FortQueryTest_GoalFrustrationDiscouragement_1 {
    class=FortQueryTest_GoalFrustrationDiscouragement
    properties {
     bDefineReferenceValue=false
     bScoreEncounterGoals=true
     bScoreEnemies=true
     bScoreSpecificAssignments=false
     bScoreWorldGoals=true
     boolValue=true
     clampMaxType=SpecifiedValue
     clampMinType=SpecifiedValue
     filterType=Maximum
     floatValueMax=1
     floatValueMin=0
     multipleContextFilterOp=AllPass
     multipleContextScoreOp=AverageScore
     normalizationType=Absolute
     referenceValue=0
     scoreClampMax=1
     scoreClampMin=0
     scoringEquation=Constant
     scoringFactor=-1
     testComment="When a Goal can't be reached, we can choose to Frustrate on (IGNORE) that Goal for a period of time."
     testPurpose=FilterAndScore
     assignmentIds []
     goalActorTagQuery {autoDescription="", tokenStreamVersion=0, userDescription="", queryTokenStream [], tagDictionary []}
     prohibitedAssignmentIds []
    }
   }
   EnvQueryTest_Dot_1 {
    class=EnvQueryTest_Dot
    properties {
     clampMaxType=SpecifiedValue
     clampMinType=SpecifiedValue
     scoreClampMax=0
     scoreClampMin=0
     testComment=Prefer Goal in front
     lineA {dirMode=TwoPoints, lineFrom=&EnvQueryContext_Querier, lineTo=&EnvQueryContext_Item}
     lineB {dirMode=Rotation, lineTo=&FortQueryContext_Goal, rotation=&EnvQueryContext_Querier}
     scoringFactor {dataBinding=null, dataField=FloatValue, defaultValue=1}
    }
   }
   FortQueryTest_GoalDistance_5 {
    class=FortQueryTest_GoalDistance
    properties {
     bScoreEncounterGoals=true
     bScoreEnemies=true
     bScoreSpecificAssignments=false
     bScoreWorldGoals=true
     clampMaxType=None
     clampMinType=None
     filterType=Range
     floatValueMax=0
     scoreClampMax=0
     scoreClampMin=0
     scoringEquation=InverseLinear
     scoringFactor=1
     testComment=Prefer closer targets.
     testPurpose=Score
     assignmentIds []
    }
   }
   FortQueryTest_GoalDistance_6 {
    class=FortQueryTest_GoalDistance
    properties {
     bScoreEncounterGoals=true
     bScoreEnemies=true
     bScoreSpecificAssignments=false
     bScoreWorldGoals=true
     clampMaxType=None
     clampMinType=None
     distanceTo=&EnvQueryContext_BuzzCool_Clone_LeashCenter_C
     filterType=Maximum
     scoreClampMax=0
     scoreClampMin=0
     scoringEquation=Linear
     scoringFactor=1
     testComment=Leash: ignore outside the leash.
     testMode=Distance2D
     testPurpose=Filter
     assignmentIds []
     floatValueMax {
      dataBinding=&self:FortQueryTest_GoalDistance_6.FortAIDataProvider_ScalableFloat_0
      dataField=FloatValue
      defaultValue=0
     }
    }
   }
  ]
 }
 {
  generator {
   name=FortQueryGenerator_GoalTrackableAIObjects_0
   class=FortQueryGenerator_GoalTrackableAIObjects
   properties {
    bAutoSortTests=true
    bCanRunAsync=false
    envQueryResultNormalizationOption=Default
    optionName=""
    searchCenter=&EnvQueryContext_Querier
    searchedActorClass=/BuzzCool/Gameplay/BGA_WildEstate_BuzzCool.BGA_WildEstate_BuzzCool_C
    assignmentIdentifier {assignmentTeam=Monster, assignmentType=World, assignmentGameplayTags {gameplayTags []}}
    requiredTag {tagName=Item.BuzzCool.Deployed}
    searchRadius {
     dataBinding=&self:FortQueryGenerator_GoalTrackableAIObjects_0.FortAIDataProvider_ScalableFloat_0
     dataField=FloatValue
     defaultValue=0
    }
   }
  }
  tests []
 }
 {
  generator {
   name=FortQueryGenerator_GoalPlayerPawns_0
   class=FortQueryGenerator_GoalPlayerPawns
   properties {bAutoSortTests=true, bCanRunAsync=false, bOnlyAthenaGameParticipants=true, envQueryResultNormalizationOption=Default, optionName=Item Holder}
  }
  tests [
   FortQueryTest_GoalGameplayTags_4 {
    class=FortQueryTest_GoalGameplayTags
    properties {
     bDefineReferenceValue=false
     bRequireAllProvidedTagQueriesPass=false
     bScoreEncounterGoals=true
     bScoreEnemies=true
     bScoreSpecificAssignments=false
     bScoreWorldGoals=true
     bShouldLookupQueryByTag=false
     bShouldPassWhenQueryNotFound=true
     boolValue=true
     clampMaxType=None
     clampMinType=None
     filterType=Match
     floatValueMax=0
     floatValueMin=0
     multipleContextFilterOp=AllPass
     multipleContextScoreOp=AverageScore
     normalizationType=Absolute
     referenceValue=0
     scoreClampMax=0
     scoreClampMin=0
     scoringEquation=Constant
     scoringFactor=1
     testComment=""
     testPurpose=Filter
     assignmentIds []
     goalActorTagQuery {autoDescription="", tokenStreamVersion=0, userDescription="", queryTokenStream [], tagDictionary []}
     pawnOwnerTagQuery {autoDescription="", tokenStreamVersion=0, userDescription="", queryTokenStream [], tagDictionary []}
     prohibitedAssignmentIds []
     queryLookupTag {tagName=None}
     tagQueryToMatch {
      autoDescription=" ANY( Ability.BuzzCool.Item.Owner )"
      tokenStreamVersion=0
      userDescription=""
      queryTokenStream [0,1,1,1,0]
      tagDictionary [{tagName=Ability.BuzzCool.Item.Owner}]
     }
    }
   }
   FortQueryTest_GoalDistance_9 {
    class=FortQueryTest_GoalDistance
    properties {
     bScoreEncounterGoals=true
     bScoreEnemies=true
     bScoreSpecificAssignments=false
     bScoreWorldGoals=true
     clampMaxType=None
     clampMinType=None
     filterType=Maximum
     scoreClampMax=0
     scoreClampMin=0
     scoringEquation=Linear
     scoringFactor=1
     testComment=""
     testPurpose=Filter
     assignmentIds []
     floatValueMax {
      dataBinding=&self:FortQueryTest_GoalDistance_9.FortAIDataProvider_ScalableFloat_0
      dataField=FloatValue
      defaultValue=0
     }
    }
   }
  ]
 }
]
